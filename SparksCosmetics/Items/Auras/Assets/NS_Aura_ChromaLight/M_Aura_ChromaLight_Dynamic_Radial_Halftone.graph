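# Translucent unlit Niagara sprite material for the ChromaLight aura, drawing a radial halftone dot pattern tinted by particle color.
# Panning caustic noise on radial UVs slightly hue-shifts the color and modulates dot size and opacity.
# Circle SDF dots tile in radial space, offset by dynamic parameter pans.
# Sphere, vertical-gradient and circle masks shape opacity, with depth fade, particle alpha and an optional center cutout scaled by dynamic parameter.
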Material /SparksCosmetics/Items/Auras/Assets/NS_Aura_ChromaLight/M_Aura_ChromaLight_Dynamic_Radial_Halftone
graph {
 properties {
  bAutomaticallySetUsageInEditor=false
  bUsedWithNiagaraSprites=true
  blendMode=BLEND_Translucent
  shadingModel=MSM_Unlit
  usedShadingModels=Unlit
 }
 nodes [
  Constant() {r=0.1}
  n1 = MaterialFunctionCall!(Center, Maintain Aspect Ratio=Constant() {r=1}, Tiling=Constant2Vector() {g=2, r=1}, Offset, In Tex Coords) {materialFunction=/Game/Effects/Fort_Effects/MaterialFunctions/MF_Uniform_Radial_UV}
  n2 = Multiply(A=@n1.Out UV, B=ScalarParameter() {defaultValue=0.5, parameterName=Uniform Noise Scale}) {constB=0.5}
  n3 = Time()
  n4 = TextureSample(UVs=Frac(None=Panner(Coordinate=@n2, Time=@n3) {speedX=-0.1, speedY=0.03})) {desc="Hue Shift Texture ", samplerType=SAMPLERTYPE_LinearGrayscale, texture=/Game/Effects/Fort_Effects/Textures/Tile/T_Caustic_Wind_Diagonal}
  n5 = OneMinus(None=@n4.R)
  n6 = ScalarParameter() {defaultValue=0.1, parameterName=hue shift perecentage}
  n7 = ParticleColor()
  MaterialFunctionCall(Emissive=MaterialFunctionCall(Hue Shift Percentage=Multiply(A=@n5, B=@n6) {constB=0.2}, Texture=@n7.RGB) {materialFunction=&HueShift}.Result) {materialFunction=&MF_FX_EmissiveAdjustments}
  n8 = MaterialFunctionCall!(Hardness=Constant() {r=0.3}, Center, Radius=Constant() {r=0.3}) {materialFunction=/Game/Packages/Fortress_SharedMaterials/Base_Material_Functions/FX_SphereMask2D}
  n9 = OneMinus(None=@n8.Result)
  n10 = Max(A=@n4.R) {constB=0.1}
  n11 = OneMinus(None=ComponentMask(None=TextureCoordinate()) {g=true})
  n12 = Multiply(
   A=MaterialFunctionCall(Hardness=Constant() {r=0.3}, Radius=Constant() {r=0.5}) {materialFunction=&FX_SphereMask2D}.Result
   B=Saturate(None=SmoothStep(Value=@n11) {constMax=0.6, constMin=0.05})
  )
  n13 = Multiply(A=@n1.Out UV, B=ScalarParameter() {defaultValue=28, desc="Tiling ", parameterName=UV Tiling})
  n14 = MaterialFunctionCall(Float2=@n13) {materialFunction=&BreakOutFloat2Components}
  n15 = DynamicParameter() {defaultValue=<rgba 0.612 1 1 1>, paramNames [Center Mask Scale, Pan X, Pan Y, Param4]}
  n16 = AppendVector(A=Add(A=@n14.R, B=@n15.Pan X), B=Add(A=@n14.G, B=@n15.Pan Y))
  n17 = Add(A=@n10) {constB=0.2}
  n18 = Saturate(None=Subtract(A=@n12, B=Multiply(A=@n8.Result) {constB=0.7}))
  n19 = MaterialFunctionCall!(
   x=MaterialFunctionCall!(In Texcoords=Frac(None=@n16)) {materialFunction=/Game/Effects/Fort_Effects/MaterialFunctions/MF_Circle_SDF}.Result
   size=Saturate(None=Multiply(A=@n17, B=@n18))
  ) {materialFunction=/Game/Effects/Fort_Effects/MaterialFunctions/MF_Fill}
  n20 = MaterialFunctionCall(In Texcoords=TextureCoordinate()) {materialFunction=&MF_Circle_SDF}
  n21 = MaterialFunctionCall(x=@n20.Result, size=@n15.Center Mask Scale) {materialFunction=&MF_Fill}
  n22 = Constant() {r=0.95}
  n23 = Multiply(
   A=LinearInterpolate(A=OneMinus(None=@n21.Result), Alpha=ScalarParameter() {defaultValue=1, desc="0  to use mask, 1 to disable", parameterName=Use Center Mask}) {constAlpha=0}
   B=Saturate(None=MaterialFunctionCall(x=@n20.Result, size=@n22) {materialFunction=&MF_Fill}.Result)
  )
  n24 = Multiply(A=Saturate(None=Add(A=@n10) {constB=0.5}), B=Multiply(A=@n12, B=Multiply(A=@n19.Result, B=@n23)))
  Multiply(
   A=Multiply(A=@n7.A, B=Multiply(A=@n9, B=@n24))
   B=DepthFade(FadeDistance=ScalarParameter() {defaultValue=50, parameterName=Depth Fade Distance})
  )
  Panner(Coordinate=@n13, Time=Time()) {speedX=-2, speedY=0.5}
 ]
}
