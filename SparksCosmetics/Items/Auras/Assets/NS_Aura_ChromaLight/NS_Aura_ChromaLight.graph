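# Niagara system for the ChromaLight cosmetic aura: a color-cycling glow that follows the character's skeletal mesh, with sizes and brightness adjusted by large-body and front-end flags.
# Chroma_Dots spawns sprites placed between sampled bones, pushed outward from a center bone and smoothed over time.
# GlowandDecals adds per-bone glows, Radial_Halftone_Backer a pulsing backdrop, and two ray emitters sprite and mesh rays.
# Engine time indexes rainbow curves, and time-of-day scalars tune glow and ray brightness.
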NiagaraSystem /SparksCosmetics/Items/Auras/Assets/NS_Aura_ChromaLight/NS_Aura_ChromaLight
dependencies {
 usedDataInterfaces [
  &NiagaraDataInterfaceArrayFloat
  &NiagaraDataInterfaceColorCurve
  &NiagaraDataInterfaceArrayInt32
  &NiagaraDataInterfaceCurve
  &NiagaraDataInterfaceArrayFloat3
  &NiagaraDataInterfaceMeshRendererInfo
  &NiagaraDataInterfaceSkeletalMesh
 ]
 usedDynamicInputs [&MakeCustomLinearColorFromBool, &RandomBool]
 usedModules [
  &self:SystemSpawnScript.SystemScriptSource.SystemScriptGraph.NiagaraNodeAssignment_1.SetVariables_#1
  &SystemState
 ]
}
emitters {
 Chroma_Dots {
  emitterUpdateScript [
   &EmitterState()
   &SpawnRate(
    Spawn Probability=0.75
    SpawnRate=&MakeCustomFloatFromBool(Bool=$User.IsLargeBody, True Float=95, False Float=80)
    Use Spawn Probability=true
   )
  ]
  particleSpawnScript [
   &PartitionParticles(Number Of Partitions=6)
   /SetVariables(
    Particles.bonesOffset=DataInterface() {propertyValues {GpuSyncMode=SyncCpuToGpu, MaxElements=0, FloatData [<-20 0 20>, <25 -5 1>, <-25 5 -1>]}}
    Particles.companion bone:NiagaraInt32=&SelectIntFromArray(Direct Array Index=$Particles.PartitionParticles.Partition, Int Selection Array=DataInterface() {propertyValues {IntData [2,1,1,2]}})
    Particles.float:NiagaraFloat=&RandomRangeFloat(Minimum=0, Maximum=1)
    Particles.positionlerp:NiagaraFloat=&RandomRangeFloat(Minimum=0, Maximum=1)
    Particles.postionOffset:Vector3f=&Multiply_VectorByFloat(
     Vector=&MakeCustomVectorFromBool(
      Bool=$User.IsFrontEnd
      True Vector=&RandomRangeVector(Minimum=<0 0 1>, Maximum=<0 0 2>)
      False Vector=<0 -1 1>
     )
     Float=&MakeCustomFloatFromBool(
      Bool=$User.IsLargeBody
      True Float=&RandomRangeFloat(Minimum=9, Maximum=15)
      False Float=&RandomRangeFloat(Minimum=2, Maximum=4)
     )
    )
   ) {isSetParameters=true}
   &InitializeParticle(
    Color=&ColorFromCurve(
     ColorCurve=DataInterface() {propertyValues {
       AlphaCurve {keys [<at 0 0.7 RCIM_Cubic>]}
       BlueCurve {keys [<at 0.349 0 RCIM_Cubic>, {interpMode=RCIM_Cubic, leaveTangent=0.8679, time=0.6348, value=0.0985}, <at 0.8756 0.5156>]}
       GreenCurve {keys [<at 0.349 0.0206 RCIM_Cubic>, {interpMode=RCIM_Cubic, leaveTangent=-0.0347, time=0.6348, value=0}, <at 0.8756 0>]}
       RedCurve {keys [<at 0.349 0.4219 RCIM_Cubic>, {interpMode=RCIM_Cubic, leaveTangent=-0.666, time=0.6348, value=0.1562}, <at 0.8756 0.0262>]}
      }}
     CurveIndex=&RandomRangeFloat(Minimum=0, Maximum=1)
    )
    Lifetime Max=0.75
    Lifetime Min:NiagaraFloat=&FloatFromCurve(
     FloatCurve=DataInterface() {propertyValues {Curve {keys [
         {arriveTangent=0.3469, interpMode=RCIM_Cubic, leaveTangent=0.3469, tangentMode=RCTM_User, time=0, value=0.45}
         {arriveTangent=0.3, interpMode=RCIM_Cubic, leaveTangent=0.3, time=0.3016, value=0.4624}
         <at 1 0.75 RCIM_Cubic>
        ]}}}
     CurveIndex=&RandomRangeFloat(Minimum=0, Maximum=1)
    )
    Lifetime Mode=::NewEnumerator1
    Sprite Size Mode=::NewEnumerator3
    Uniform Sprite Size=&Multiply_Float(
     A=&RandomRangeFloat(Minimum=2, Maximum=5.5)
     B=&MakeCustomFloatFromBool(Bool=$User.IsLargeBody, True Float=1.1, False Float=1)
    )
    UsePositionOffset=false
    Write Material Random=true
   )
   /SetVariables() {isSetParameters=true}
   &CameraOffset(Camera Offset Amount=-15) {module=CameraOffset001}
   &AddVelocity(
    Module.Use Falloff Distance=false
    Origin Offset=<0 0 -5>
    Origin Offset Coordinate Space=::World
    Override Seed=false
    Velocity Mode=::NewEnumerator1
    Velocity Origin=$Output.SkeletalMeshLocation001.SampledPositionWithoutOffset
    Velocity Speed=&Multiply_Float(
     A=&Multiply_Float(
      A=&Multiply_Float(
       A=&RandomRangeFloat(Minimum=65, Maximum=100)
       B=&MakeCustomFloatFromBool(Bool=$User.IsFrontEnd, True Float=1, False Float=1)
      )
      B=&MakeCustomFloatFromBool(Bool=$User.IsLargeBody, True Float=1.2, False Float=1)
     )
     B=1
    )
   )
   &AddVelocity() {module=AddVelocity001}
   &ShapeLocation(
    MaskSpawnGroup=false
    Non Uniform Scale=<1 1 0.5>
    Override Seed=false
    Shape Origin=&VectorFromFloat(Value=0)
    Sphere Radius=7
    UseOwnerScale=true
   )
  ]
  particleUpdateScript [
   &ParticleState()
   &SkeletalMeshLocation(
    Bone / Socket Index=7
    Bone Sampling Mode=::NewEnumerator4
    MaskSpawnGroup=false
    Position Sampling=::NewEnumerator2
    Sampled Position Offset=<0 0 -10>
    Sampled Position Offset Coordinate Space:ENiagaraSkelMeshTransforms=::NewEnumerator2
    Skeletal Mesh=$System.Skeletal Mesh
   ) {module=SkeletalMeshLocation001}
   &SkeletalMeshLocation(
    Bone / Socket Index=&SelectIntFromArray(Direct Array Index=$Particles.PartitionParticles.Partition, Int Selection Array=DataInterface() {propertyValues {IntData [2,1,1,2,2,2]}})
    Bone Sampling Mode=::NewEnumerator4
    MaskSpawnGroup=false
    Orientation Sampling=::NewEnumerator1
    Position Sampling=::NewEnumerator2
    Sampled Position Offset:NiagaraPosition=$Particles.ShapeLocation.ShapePosition
    Skeletal Mesh=$System.Skeletal Mesh
   ) {module=SkeletalMeshLocation002}
   &SkeletalMeshLocation(
    Bone / Socket Index=&SelectIntFromArray(Direct Array Index=$Particles.PartitionParticles.Partition, Int Selection Array=DataInterface() {propertyValues {IntData [1,2,3,4,1,1]}})
    Bone Sampling Mode=::NewEnumerator4
    MaskSpawnGroup=false
    Position Sampling=::NewEnumerator2
    Sampled Position Offset:NiagaraPosition=$Particles.ShapeLocation.ShapePosition
    Skeletal Mesh=$System.Skeletal Mesh
   )
   &ScaleColor(
    Linear Color Curve=DataInterface() {propertyValues {
      AlphaCurve {keys [<at 0.6188 1>,<at 1 0>]}
      BlueCurve {keys [<at 0 1>,<at 1 1>]}
      GreenCurve {keys [<at 0 1>,<at 1 0.0009>]}
      RedCurve {keys [<at 0 1>,<at 1 0.7443>]}
     }}
    Scale Mode=::NewEnumerator2
   )
   &ScaleSpriteSize(Uniform Curve Sprite Scale=DataInterface() {propertyValues {Curve {keys [
        {arriveTangent=16.2624, interpMode=RCIM_Cubic, leaveTangent=16.2624, leaveTangentWeight=0.7905, tangentMode=RCTM_User, time=0, value=0.3}
        {arriveTangent=-0.3, interpMode=RCIM_Cubic, leaveTangent=-0.3, time=0.1, value=1}
        {arriveTangent=-0.4667, interpMode=RCIM_Cubic, leaveTangent=-0.4667, tangentMode=RCTM_User, time=1, value=0}
       ]}}}) {module=ScaleSpriteSize001}
   &ScaleSpriteSize(
    Scale Factor=&MakeVector2D(
     X=&FloatFromCurve(FloatCurve=DataInterface() {propertyValues {Curve {keys [
          {arriveTangent=-1.7833, interpMode=RCIM_Cubic, leaveTangent=-1.7833, tangentMode=RCTM_User, time=0, value=1}
          <at 0.15 0.6706 RCIM_Cubic>
          <at 0.7 1 RCIM_Cubic>
         ]}}})
     Y=&FloatFromCurve(FloatCurve=DataInterface() {propertyValues {Curve {keys [
          {arriveTangent=1.1521, interpMode=RCIM_Cubic, leaveTangent=1.1521, tangentMode=RCTM_User, time=0.002, value=1}
          <at 0.15 1.1845 RCIM_Cubic>
          <at 0.7 1 RCIM_Cubic>
         ]}}})
    )
    Scale Sprite Size Mode=::NewEnumerator2
   ) {module=ScaleSpriteSize002}
   &Drag(
    Drag=&FloatFromCurve(
     FloatCurve=DataInterface() {propertyValues {Curve {keys [<at 0.1 4>,<at 0.15 5.5>,<at 0.5011 6>,<at 0.8 7>]}}}
     CurveIndex=&RandomRangeFloat(Minimum=0, Maximum=1)
    )
    Use Linear Drag=true
    Use Rotational Drag=false
   ) {module=Drag001}
   &SolveForcesAndVelocity(Clamp Velocity=false, Limit Acceleration=false)
   /SetVariables(Particles.Position=&LerpPosition(A=$Particles.SkeletalMeshLocation.SampledPosition, B=$Particles.SkeletalMeshLocation002.SampledPosition, Alpha=$Particles.positionlerp)) {isSetParameters=true}
   &DynamicMaterialParameters(
    Index 0 Param 1=&FloatFromCurve(
     FloatCurve=DataInterface() {propertyValues {Curve {defaultValue=1, keys [<at 0.5 0.05>, {tangentMode=RCTM_User, time=0.7, value=0.05}, {tangentMode=RCTM_User, time=1, value=1}]}}}
     Scale Curve=&SelectFloatFromArray(
      Array Sampling Mode=::NewEnumerator0
      Direct Array Index=?
      Float Selection Array=DataInterface() {propertyValues {FloatData [0,2,2]}}
      Random Seed=0
     )
    )
    Index 0 Param 2=&FloatFromCurve(FloatCurve=DataInterface() {propertyValues {Curve {
        defaultValue=1
        keys [
         {arriveTangent=8.2781, interpMode=RCIM_Cubic, leaveTangent=8.2781, tangentMode=RCTM_User, time=0.2, value=0}
         {arriveTangent=1.1132, interpMode=RCIM_Cubic, leaveTangent=1.1132, tangentMode=RCTM_User, time=0.8, value=3}
        ]
       }}})
    Module.Index0Float0Write=true
    Module.Index0Float1Write=true
    Module.Index0Float2Write=true
    Module.Index0Float3Write=true
   )
   /SetVariables(
    Particles.Pivotoffset2:Vector2f=&Vector2DFromFloat(Value=&FloatFromCurve(FloatCurve=DataInterface() {propertyValues {Curve {keys [
          {arriveTangent=0.3502, interpMode=RCIM_Cubic, leaveTangent=0.3502, tangentMode=RCTM_User, time=0, value=0.5}
          {arriveTangent=2.7258, interpMode=RCIM_Cubic, leaveTangent=2.7258, tangentMode=RCTM_User, time=1, value=2}
         ]}}}))
    Particles.color2:LinearColor=&ColorFromCurve(ColorCurve=DataInterface() {propertyValues {
       AlphaCurve {keys [<at 0 0 RCIM_Cubic>,<at 0.178 0.55 RCIM_Cubic>,<at 0.9937 0>]}
       BlueCurve {keys [
         <at 0 0 RCIM_Cubic>
         {arriveTangent=0.4334, interpMode=RCIM_Cubic, leaveTangent=0.4334, time=0.306, value=0}
         <at 0.5906 0.2559>
         <at 0.9454 0.9011>
        ]}
       GreenCurve {keys [<at 0 0.0019 RCIM_Cubic>, {arriveTangent=1.69, interpMode=RCIM_Cubic, leaveTangent=1.69, time=0.306, value=1}, <at 0.5906 1>, <at 0.9454 1>]}
       RedCurve {keys [
         <at 0 1 RCIM_Cubic>
         {arriveTangent=-1.6933, interpMode=RCIM_Cubic, leaveTangent=-1.6933, time=0.306, value=0.5257}
         <at 0.5906 0>
         <at 0.9454 0>
        ]}
      }})
   ) {isSetParameters=true}
   /SetVariables() {isSetParameters=true}
   /SetVariables(
    Particles.Velocity=&SubtractVectorFromPosition(
     Position=$Particles.Position
     Vector=Linked() {linkedVariable {name=Particles.SkeletalMeshLocation001.SampledPosition, type=NiagaraPosition}}
    )
    Particles.position modified:NiagaraPosition=&AddVectorToPosition(
     Position=$Particles.Position
     Vector=&Multiply_VectorByFloat(
      Vector=&NormalizeVector(
       Vector To Normalize=&SubtractVectorFromPosition(
        Position=$Particles.Position
        Vector=Linked() {linkedVariable {name=Particles.SkeletalMeshLocation001.SampledPosition, type=NiagaraPosition}}
       )
       Safe Normalize=true
       Safe Fallback Threshold=0
       Safe Fallback Vector=<0 0 0.001>
      )
      Float=&FloatFromCurve(
       FloatCurve=DataInterface() {propertyValues {Curve {
          defaultValue=0
          keys [
           {arriveTangent=2.7405, interpMode=RCIM_Cubic, leaveTangent=2.7405, tangentMode=RCTM_User, time=0, value=0}
           {arriveTangent=0.4924, interpMode=RCIM_Cubic, leaveTangent=0.4924, tangentMode=RCTM_User, time=0.65, value=0.9}
           {arriveTangentWeight=0.2204, interpMode=RCIM_Cubic, tangentMode=RCTM_User, time=0.9966, value=1}
          ]
         }}}
       Scale Curve=&Multiply_Float(
        A=&Multiply_Float(
         A=&FloatFromCurve(
          FloatCurve=DataInterface() {propertyValues {Curve {keys [<at 0.3 18>,<at 0.4 25>,<at 0.85 30>,<at 1 45>]}}}
          CurveIndex=&RandomRangeFloat(Minimum=0, Maximum=1)
         )
         B=&SelectFloatFromArray(
          Direct Array Index=$Particles.PartitionParticles.Partition
          Float Selection Array=DataInterface() {propertyValues {FloatData [1.5,1.6,0.45,0.45,1.1,1.25]}}
         )
        )
        B=&MakeCustomFloatFromBool(Bool=$User.IsLargeBody, True Float=1.1, False Float=1)
       )
      )
     )
    )
   ) {isSetParameters=true}
   &InterpolateOverTime(
    Rate Of Change:NiagaraFloat=&FloatFromCurve(
     FloatCurve=DataInterface() {propertyValues {Curve {keys [<at 0.3 1 RCIM_Cubic>, {arriveTangentWeight=0.5069, interpMode=RCIM_Cubic, tangentMode=RCTM_User, time=1, value=0}]}}}
     Scale Curve=8
    )
    Target Position Value:NiagaraPosition=$Particles.position modified
    Variable Type=::NewEnumerator6
   )
   /SetVariables(Particles.Position=$StackContext.InterpolateOverTime.InterpolateOverTime_Position.Moving Average) {isSetParameters=true}
  ]
  rendererClasses [&NiagaraSpriteRendererProperties]
  renderers [
   {rendererClass=&NiagaraSpriteRendererProperties}
   {rendererClass=&NiagaraSpriteRendererProperties, rendererIndex=1}
  ]
 }
 GlowandDecals {
  emitterUpdateScript [
   &EmitterState()
   &SpawnBurst_Instantaneous(Spawn Count=7, Use Spawn Probability=false, UseLoopCountLimit=true)
  ]
  particleSpawnScript [
   &PartitionParticles(Number Of Partitions=7)
   &InitializeParticle(
    Color=<rgba 6 6 6 1>
    Lifetime=3
    Sprite Rotation Mode=::NewEnumerator0
    Sprite Size Mode=::NewEnumerator3
    Uniform Sprite Size=&Multiply_Float(
     A=&SelectFloatFromArray(
      Direct Array Index=&SpawnGroup()
      Float Selection Array=DataInterface() {propertyValues {FloatData [40,40,20,20,30,30]}}
     )
     B=&MakeCustomFloatFromBool(Bool=$User.IsLargeBody, True Float=1.5, False Float=1)
    )
    UsePositionOffset=false
    Write Material Random=true
   ) {module=InitializeParticle001}
   /SetVariables(Particles.glowScale:Vector2f=&Vector2DFromFloat(Value=&Multiply_Float(
      A=&SelectFloatFromArray(
       Direct Array Index=$Particles.PartitionParticles.Partition
       Float Selection Array=DataInterface() {propertyValues {FloatData [35,45,70,95,70,45,35]}}
      )
      B=&MakeCustomFloatFromBool(Bool=$User.IsLargeBody, True Float=1.2, False Float=1)
     ))) {isSetParameters=true}
   /SetVariables(Particles.glowCameraOffset=-20) {isSetParameters=true}
   /SetVariables() {isSetParameters=true}
  ]
  particleUpdateScript [
   &ParticleState(
    Kill Particles When Lifetime Has Elapsed=false
    Let Infinitely Lived Particles Die When Emitter Deactivates=true
    Loop Particles Lifetime=true
   )
   /SetVariables(Particles.RainbowHueShift:LinearColor=&ColorFromCurve(
     ColorCurve=DataInterface() {propertyValues {
       AlphaCurve {keys []}
       BlueCurve {keys [
         <at 0 0 RCIM_Cubic>
         {arriveTangent=0.5177, interpMode=RCIM_Cubic, leaveTangent=0.5177, time=0.2331, value=0}
         <at 0.3655 0.1892>
         <at 0.4997 0.6568>
         <at 0.6684 1>
         <at 0.8969 1>
         <at 1 0>
        ]}
       GreenCurve {keys [
         <at 0 1 RCIM_Cubic>
         {arriveTangent=-2.7361, interpMode=RCIM_Cubic, leaveTangent=-2.7361, time=0.2331, value=0.1937}
         <at 0.3655 0>
         <at 0.4997 0.0005>
         <at 0.6684 0.3419>
         <at 0.8969 0.8482>
         <at 1 1>
        ]}
       RedCurve {keys [
         <at 0 0.1878 RCIM_Cubic>
         {arriveTangent=2.2222, interpMode=RCIM_Cubic, leaveTangent=2.2222, time=0.2331, value=1}
         <at 0.3655 1>
         <at 0.4997 0.1835>
         <at 0.6684 0>
         <at 0.8969 0.1742>
         <at 1 0.1878>
        ]}
      }}
     CurveIndex=&Frac_Float(Float=&Add_Float(A=&Multiply_Float(A=$Engine.Time, B=0.2), B=&ReturnNormalizedExecIndex()))
    )) {isSetParameters=true}
   &CameraOffset(Camera Offset Amount=5)
   &SpriteRotationRate(Rotation Rate=&RandomRangeFloat(Minimum=10, Maximum=20))
   &ScaleColor(
    Scale Alpha=&MakeCustomFloatFromBool(Bool=$User.IsFrontEnd, True Float=0.15, False Float=0.35)
    Scale RGB=&Multiply_VectorByFloat(
     Vector=&MakeVectorFromLinearColorRGB(Color=<rgba 1 0 0.0312 1>)
     Float=&FNTimeOfDayScalar(Night Scale=1.1, Morning Scale=1, Day Scale=1.1, Evening Scale=1.1)
    )
    ScaleA=true
    ScaleRGB=true
   ) {module=ScaleColor001}
   &SkeletalMeshLocation(
    Bone / Socket Index=&SelectIntFromArray(Direct Array Index=$Particles.PartitionParticles.Partition, Int Selection Array=DataInterface() {propertyValues {IntData [5,3,1,0,2,4,6]}})
    Bone Sampling Mode=::NewEnumerator4
    MaskSpawnGroup=false
    Orientation Sampling=::NewEnumerator1
    Sampled Position Offset Coordinate Space:ENiagaraSkelMeshTransforms=::NewEnumerator1
    Skeletal Mesh=$System.Skeletal Mesh
   ) {module=SkeletalMeshLocation003}
   &DynamicMaterialParameters(
    Index 0 Param 1=&SelectFloatFromArray(Direct Array Index=$Particles.PartitionParticles.Partition, Float Selection Array=DataInterface() {propertyValues {FloatData [5,5,4,4,3,3]}})
    Index 0 Param 3=1
    Module.Index0Float0Write=true
    Module.Index0Float1Write=true
    Module.Index0Float2Write=false
    Module.Index0Float3Write=true
   )
   /SetVariables(Particles.glowColor:LinearColor=&MakeLinearColorFromVectorAndFloat(
     "Vector (RGB)"=&Multiply_VectorByFloat(
      Vector=&MakeVectorFromLinearColorRGB(Color=<rgba 1 0.0364 0 1>)
      Float=&SelectFloatFromArray(
       Direct Array Index=$Particles.PartitionParticles.Partition
       Float Selection Array=DataInterface() {propertyValues {FloatData [2.5,2.5,3,3,3,2.5,2.5]}}
      )
     )
     "Float (Alpha)"=&FloatFromCurve(
      FloatCurve=DataInterface() {propertyValues {Curve {keys [<at 0 1>]}}}
      Scale Curve=&MakeCustomFloatFromBool(Bool=$User.IsFrontEnd, True Float=0.2, False Float=0.3)
     )
    )) {isSetParameters=true}
  ]
  rendererClasses [&NiagaraSpriteRendererProperties]
  renderers [
   {rendererClass=&NiagaraSpriteRendererProperties}
   {rendererClass=&NiagaraSpriteRendererProperties, rendererIndex=1}
  ]
 }
 Radial_Halftone_Backer {
  emitterUpdateScript [
   &EmitterState(
    Inactive Response=::NewEnumerator1
    Life Cycle Mode=::NewEnumerator1
    Loop Behavior=::NewEnumerator1
    Loop Duration Mode=::NewEnumerator1
    UseLoopDelay=false
   )
   &SpawnBurst_Instantaneous(Use Spawn Probability=false, UseLoopCountLimit=true)
  ]
  particleSpawnScript [&InitializeParticle(
    Color=<rgba 1 1 1 0.5>
    Lifetime=5
    Sprite Size Mode=::NewEnumerator3
    Uniform Sprite Size=&Multiply_Float(A=160, B=&MakeCustomFloatFromBool(Bool=$User.IsLargeBody, True Float=1.1, False Float=0.9))
    UsePositionOffset=false
    Write Material Random=true
   ) {module=InitializeParticle001}]
  particleUpdateScript [
   &ParticleState(
    Kill Particles When Lifetime Has Elapsed=false
    Let Infinitely Lived Particles Die When Emitter Deactivates=true
    Loop Particles Lifetime=true
   )
   &CameraOffset(Camera Offset Amount=-20)
   &ScaleSpriteSize(
    Scale Sprite Size Mode=::NewEnumerator0
    Uniform Scale Factor=&Waveform("Waveform [1]"=::NewEnumerator12, "[1] Frequency"=0.25, "[1] Amplitude Min/Max"=<0.9 1.2>)
   )
   /SetVariables(
    Particles.Linear Color=<rgba 0 0.0026 0.0521 0.6>
    Particles.Vector 2D:Vector2f=&Multiply_Vector2D_ByFloat(Float=1.03, Vector2D=$Particles.SpriteSize)
   ) {isSetParameters=true}
   &SkeletalMeshLocation(
    Bone / Socket Index=7
    Bone Sampling Mode=::NewEnumerator4
    MaskSpawnGroup=false
    Sampled Position Offset=&Multiply_VectorByFloat(
     Vector=<1 -1 0>
     Float=&Waveform("Waveform [1]"=::NewEnumerator0, "[1] Frequency"=0.1, "[1] Amplitude Min/Max"=<-5 5>)
    )
    Skeletal Mesh=$System.Skeletal Mesh
   ) {module=SkeletalMeshLocation001}
   &InterpolateOverTime(Rate Of Change=15, Target Position Value:NiagaraPosition=$Particles.SkeletalMeshLocation001.SampledPosition, Variable Type=::NewEnumerator6)
   &DynamicMaterialParameters(
    Index 0 Param 1=0.344
    Index 0 Param 2=&FloatFromCurve(
     FloatCurve=DataInterface() {propertyValues {Curve {defaultValue=1, keys [{tangentMode=RCTM_User, time=0, value=1}, {tangentMode=RCTM_User, time=1, value=0}]}}}
     CurveIndex=&Frac_Float(Float=&Multiply_Float(A=$Engine.Time, B=2.5))
    )
    Index 0 Param 3=&FloatFromCurve(
     FloatCurve=DataInterface() {propertyValues {Curve {
        defaultValue=0
        keys [
         {interpMode=RCIM_Cubic, tangentMode=RCTM_User, time=0, value=0}
         {interpMode=RCIM_Cubic, tangentMode=RCTM_User, time=0.5, value=1}
         {tangentMode=RCTM_User, time=1, value=0}
        ]
       }}}
     CurveIndex=&Frac_Float(Float=&Multiply_Float(A=$Engine.Time, B=0.2))
    )
    Module.Index0Float0Write=true
    Module.Index0Float1Write=true
    Module.Index0Float2Write=true
    Module.Index0Float3Write=true
   )
   &ScaleColor(
    Scale Alpha=&FloatFromCurve(
     FloatCurve=DataInterface() {propertyValues {Curve {
        defaultValue=1
        keys [
         {interpMode=RCIM_Cubic, tangentMode=RCTM_User, time=0, value=0.7}
         {interpMode=RCIM_Cubic, tangentMode=RCTM_User, time=0.5, value=1}
         <at 1 0.7 RCIM_Cubic>
        ]
       }}}
     CurveIndex=&Frac_Float(Float=&Multiply_Float(A=$Engine.Time, B=0.4))
    )
    Scale RGB=&MakeVectorFromLinearColorRGB(Color=&ColorFromCurve(
      ColorCurve=DataInterface() {propertyValues {
        AlphaCurve {keys []}
        BlueCurve {keys [
          <at 0.0245 0.0319>
          <at 0.1834 1>
          <at 0.2697 0.5089>
          {arriveTangent=1.0565, interpMode=RCIM_Cubic, leaveTangent=1.0565, time=0.4404, value=0}
          <at 0.6208 0.5089>
          <at 0.7346 1>
          <at 0.9405 0.0319>
         ]}
        GreenCurve {keys [
          <at 0.0245 0>
          <at 0.1834 0.3419>
          <at 0.2697 0.5395>
          {arriveTangent=-0.425, interpMode=RCIM_Cubic, leaveTangent=-0.425, time=0.4404, value=0.7604}
          <at 0.6208 0.5395>
          <at 0.7346 0.3419>
          <at 0.9405 0>
         ]}
        RedCurve {keys [<at 0.0245 1>,<at 0.1834 0>,<at 0.2697 0>,<at 0.4404 0.5684 RCIM_Cubic>,<at 0.6208 0>,<at 0.7346 0>,<at 0.9405 1>]}
       }}
      CurveIndex=&Frac_Float(Float=&Multiply_Float(A=$Engine.Time, B=0.2))
     ))
    ScaleA=true
    ScaleRGB=true
   )
   /SetVariables(Particles.Position=$Particles.InterpolateOverTime.InterpolateOverTime_Position.Moving Average) {isSetParameters=true}
  ]
  rendererClasses [&NiagaraSpriteRendererProperties]
  renderers [
   {rendererClass=&NiagaraSpriteRendererProperties}
   {rendererClass=&NiagaraSpriteRendererProperties, rendererIndex=1}
  ]
 }
 Rays {
  emitterUpdateScript [
   &EmitterState()
   &SpawnBurst_Instantaneous(Spawn Group=1, Use Spawn Probability=false, UseLoopCountLimit=true) {module=SpawnBurst_Instantaneous001}
   &SpawnBurst_Instantaneous(Use Spawn Probability=false, UseLoopCountLimit=true)
  ]
  particleSpawnScript [
   &InitializeParticle(
    Color=<rgba 1.25 1.25 1.25 1>
    Mesh Scale Mode=::NewEnumerator3
    Mesh Uniform Scale=&MakeCustomFloatFromBool(Bool=$User.IsLargeBody, True Float=1.1, False Float=0.85)
    Sprite Size=&Multiply_Vector2D_ByFloat(
     Float=&MakeCustomFloatFromBool(Bool=$User.IsLargeBody, True Float=174, False Float=145)
     Vector2D=<1 1.1>
    )
    Sprite Size Mode=::NewEnumerator4
    UsePositionOffset=false
    Write Material Random=true
   )
   /SetVariables(Particles.RaySpriteSize:Vector2f=&Vector2DFromFloat(Value=&MakeCustomFloatFromBool(Bool=$User.IsLargeBody, True Float=135, False Float=110))) {isSetParameters=true}
   &DynamicMaterialParameters(
    Index 0 Param 1=&RandomRangeFloat(Minimum=0, Maximum=10)
    Index 0 Param 2=&RandomRangeFloat(Minimum=1, Maximum=2)
    Index 0 Param 3=&RandomRangeFloat(Minimum=0.8, Maximum=1.6)
    Index 2 Param 1=2
    Index 2 Param 2=1
    Module.Index0Float0Write=true
    Module.Index0Float1Write=true
    Module.Index0Float2Write=true
    Module.Index0Float3Write=true
    Module.Index2Float0Write=true
    Module.Index2Float1Write=true
    Module.Index2Float2Write=true
    Module.Index2Float3Write=true
    Write Parameter Index 2=::NewEnumerator1
   )
   &SkeletalMeshLocation(
    Bone / Socket Index=7
    Bone Sampling Mode=::NewEnumerator4
    MaskSpawnGroup=false
    Orientation Sampling=::NewEnumerator1
    Skeletal Mesh=$System.Skeletal Mesh
   )
   &CameraOffset(Camera Offset Amount=-50)
   /SetVariables(Particles.GlowBackerColor:LinearColor=<rgba 0.0003 0.0048 0.1042 1.75>) {isSetParameters=true}
   &InitialMeshOrientation(Mesh Orientation Mode=::NewEnumerator2, Rotation=<0.5 0 0>, Use Rotation Vector=true)
   /SetVariables(Particles.VisibilityTag=&SpawnGroup()) {isSetParameters=true}
  ]
  particleUpdateScript [
   /SetVariables(Particles.RainbowHueShift:LinearColor=&ColorFromCurve(
     ColorCurve=DataInterface() {propertyValues {
       AlphaCurve {keys []}
       BlueCurve {keys [
         <at 0 0 RCIM_Cubic>
         {arriveTangent=0.594, interpMode=RCIM_Cubic, leaveTangent=0.594, time=0.1911, value=0}
         <at 0.3185 0.1892>
         <at 0.4997 1>
         <at 0.6684 1.1>
         <at 0.8969 1>
         <at 1 0>
        ]}
       GreenCurve {keys [
         <at 0 1 RCIM_Cubic>
         {arriveTangent=-3.1393, interpMode=RCIM_Cubic, leaveTangent=-3.1393, time=0.1911, value=0.1937}
         <at 0.3185 0>
         <at 0.4997 0.0008>
         <at 0.6684 0.3761>
         <at 0.8969 0.8482>
         <at 1 1>
        ]}
       RedCurve {keys [
         <at 0 0.1878 RCIM_Cubic>
         {arriveTangent=2.5496, interpMode=RCIM_Cubic, leaveTangent=2.5496, time=0.1911, value=1}
         <at 0.3185 1>
         <at 0.4997 0.2793>
         <at 0.6684 0>
         <at 0.8969 0.1742>
         <at 1 0.1878>
        ]}
      }}
     CurveIndex=&Frac_Float(Float=&Multiply_Float(A=$Engine.Time, B=0.2))
    )) {isSetParameters=true}
   &ParticleState(Kill Particles When Lifetime Has Elapsed=false)
   &SpriteRotationRate(Rotation Rate=10)
   &SkeletalMeshLocation(
    Bone / Socket Index=7
    Bone Sampling Mode=::NewEnumerator4
    MaskSpawnGroup=false
    Orientation Sampling=::NewEnumerator1
    Sampled Position Offset=&MaskVector(Mask Source=0, Mask Parameter=?, Mask=0, Masked Vector=<0 1 0>, Passthrough Vector=<9 1 0>)
    Skeletal Mesh=$System.Skeletal Mesh
   ) {module=SkeletalMeshLocation001}
   &ScaleColor(
    Scale Alpha=&MakeCustomFloatFromBool(Bool=$User.IsFrontEnd, True Float=0.4, False Float=0.75)
    Scale RGB=&Multiply_VectorByFloat(
     Vector=&MakeVectorFromLinearColorRGB(Color=$Particles.RainbowHueShift)
     Float=&FNTimeOfDayScalar(Night Scale=1.2, Morning Scale=1.05, Day Scale=1.1, Evening Scale=1.1)
    )
    ScaleA=true
    ScaleRGB=true
   ) {module=ScaleColor001}
   &DynamicMaterialParameters(
    Index 1 Param 1=$Engine.Time
    Index 1 Param 2=0.2
    Module.Index1Float0Write=true
    Module.Index1Float1Write=true
    Module.Index1Float2Write=true
    Module.Index1Float3Write=true
    Write Parameter Index 0=::NewEnumerator0
    Write Parameter Index 1=::NewEnumerator1
   ) {module=DynamicMaterialParameters001}
  ]
  rendererClasses [&NiagaraSpriteRendererProperties]
  renderers [
   {rendererClass=&NiagaraSpriteRendererProperties}
   {rendererClass=&NiagaraSpriteRendererProperties, rendererIndex=1}
  ]
 }
 rays001 {
  emitterUpdateScript [
   &EmitterState(
    Inactive Response=::NewEnumerator1
    Life Cycle Mode=::NewEnumerator1
    Loop Behavior=::NewEnumerator1
    Loop Duration Mode=::NewEnumerator1
    UseLoopDelay=false
   )
   &SpawnBurst_Instantaneous(Spawn Count=3, Use Spawn Probability=false, UseLoopCountLimit=false)
  ]
  particleSpawnScript [
   &PartitionParticles(Number Of Partitions=3)
   &InitializeParticle(
    Color=<rgba 3 3 3 1>
    Mesh Scale=&Multiply_VectorByFloat(
     Vector=&MakeVector(X=0.4, Y=0.4, Z=0.3)
     Float=&MakeCustomFloatFromBool(Bool=$User.IsLargeBody, True Float=1.1, False Float=0.9)
    )
    Mesh Scale Mode=::NewEnumerator4
    UsePositionOffset=false
    Write Material Random=true
   )
   &InitialMeshOrientation(
    Mesh Orientation Mode=::NewEnumerator2
    Rotation=&SelectVectorFromArray(
     Direct Array Index=$Particles.PartitionParticles.Partition
     Vector Selection Array=DataInterface() {propertyValues {FloatData [<0 0 0>, <0 0.15 0>, <0 -0.15 0>]}}
    )
    Use Rotation Vector=true
   )
   &CameraOffset(Camera Offset Amount=-20)
  ]
  particleUpdateScript [
   &ParticleState(Kill Particles When Lifetime Has Elapsed=false, Let Infinitely Lived Particles Die When Emitter Deactivates=true)
   &UpdateMeshOrientation(
    Facing Direction=&MakeVector(
     X=1
     Y=0
     Z=&FloatFromCurve(
      FloatCurve=DataInterface() {propertyValues {Curve {
         defaultValue=1
         keys [
          {interpMode=RCIM_Cubic, tangentMode=RCTM_User, time=0, value=-1}
          {interpMode=RCIM_Cubic, tangentMode=RCTM_User, time=0.5, value=1}
          {tangentMode=RCTM_User, time=1, value=-1}
         ]
        }}}
      CurveIndex=&Frac_Float(Float=&Multiply_Float(A=$Engine.Time, B=0.2))
      Scale Curve=0.3
     )
    )
    Facing Interpolation=false
    Orientation Method=::NewEnumerator1
    Override Quat=false
    TransformFacing=false
    TransformUp=false
   ) {module=UpdateMeshOrientation002}
   &ScaleMeshSize(Scale Factor=&MakeVector(
     X=&Waveform(
      Global Phase=&Add_Float(A=$Emitter.Age, B=&ReturnNormalizedExecIndex())
      "Waveform [1]"=::NewEnumerator12
      "[1] Phase Scale"=3
      "[1] Frequency"=0.5
      "[1] Amplitude Min/Max"=<1 1.6>
     )
     Y=1
     Z=1
    ))
   &DynamicMaterialParameters(
    Index 0 Param 2=0.1
    Index 0 Param 3=&RandomRangeFloat(Minimum=0, Maximum=1)
    Index 0 Param 4=&RandomRangeFloat(Minimum=0, Maximum=1)
    Module.Param0WriteEnabled=true
    Module.Param1WriteEnabled=true
    Module.Param2WriteEnabled=true
    Module.Param3WriteEnabled=true
   )
   &SkeletalMeshLocation(
    Bone / Socket Index=&SelectIntFromArray(Direct Array Index=$Particles.PartitionParticles.Partition, Int Selection Array=DataInterface() {propertyValues {IntData [0,7,7,7,7]}})
    Bone Sampling Mode=::NewEnumerator4
    "Inherit Velocity (Scale)"=-0.2
    MaskSpawnGroup=false
    Orientation Sampling=::NewEnumerator1
    Sampled Position Offset=&SelectVectorFromArray(
     Direct Array Index=$Particles.PartitionParticles.Partition
     Vector Selection Array=DataInterface() {propertyValues {FloatData [<0 0 -15>, <0 0 0>, <0 0 0>, <0 0 0>, <0 0 0>]}}
    )
    Sampled Position Offset Coordinate Space:ENiagaraSkelMeshTransforms=::NewEnumerator1
    Skeletal Mesh=$System.Skeletal Mesh
    Velocity Sampling=::NewEnumerator0
    VelocitySpeedLimit=true
   )
   &ScaleColor(
    Scale Alpha=&MakeCustomFloatFromBool(Bool=$User.IsFrontEnd, True Float=0.5, False Float=0.8)
    Scale RGB=&Multiply_VectorByFloat(
     Vector=&MakeVectorFromLinearColorRGB(Color=&ColorFromCurve(
       ColorCurve=DataInterface() {propertyValues {
         AlphaCurve {keys []}
         BlueCurve {keys [
           <at 0 0 RCIM_Cubic>
           {arriveTangent=0.5414, interpMode=RCIM_Cubic, leaveTangent=0.5414, time=0.0862, value=0}
           <at 0.3495 0.1892>
           <at 0.5627 0.6568>
           <at 0.6944 0.6615>
           <at 0.8483 0.5573>
           <at 0.9595 0>
          ]}
         GreenCurve {keys [
           <at 0 1 RCIM_Cubic>
           {arriveTangent=-2.8611, interpMode=RCIM_Cubic, leaveTangent=-2.8611, time=0.0862, value=0.1937}
           <at 0.3495 0>
           <at 0.5627 0.0005>
           <at 0.6944 0.2262>
           <at 0.8483 0.4548>
           <at 0.9595 1>
          ]}
         RedCurve {keys [
           <at 0 0.1878 RCIM_Cubic>
           {arriveTangent=2.3238, interpMode=RCIM_Cubic, leaveTangent=2.3238, time=0.0862, value=1}
           <at 0.3495 1>
           <at 0.5627 0.1835>
           <at 0.6944 0>
           <at 0.8483 0>
           <at 0.9595 0.1878>
          ]}
        }}
       CurveIndex=&Frac_Float(Float=&Add_Float(
         A=&Multiply_Float(A=$Engine.Time, B=0.2)
         B=&ReturnNormalizedExecIndex(Normalized Index Includes One=false)
        ))
      ))
     Float=&FNTimeOfDayScalar(Night Scale=1, Morning Scale=0.95, Day Scale=1.2, Evening Scale=1.1)
    )
    ScaleA=true
    ScaleRGB=true
   ) {module=ScaleColor001}
   /SetVariables() {isSetParameters=true}
  ]
  rendererClasses [&NiagaraMeshRendererProperties]
  renderers [{rendererClass=&NiagaraMeshRendererProperties, rendererIndex=0}]
 }
}
summary {
 systemName=NS_Aura_ChromaLight
 emitters [
  {emitterName=GlowandDecals, rendererClasses [&NiagaraSpriteRendererProperties]}
  {emitterName=Rays, rendererClasses [&NiagaraSpriteRendererProperties]}
  {emitterName=rays001, rendererClasses [&NiagaraMeshRendererProperties]}
  {emitterName=Chroma_Dots, rendererClasses [&NiagaraSpriteRendererProperties]}
  {emitterName=Radial_Halftone_Backer, rendererClasses [&NiagaraSpriteRendererProperties]}
 ]
 userVariables [
  User.bIsFrontEnd {type=NiagaraBool}
  User.bIsFrontEndPreview {type=NiagaraBool}
  User.Interpolate Intensity {type=NiagaraFloat, defaultValue=5, description=""}
  User.IsFrontEnd {type=NiagaraBool}
  User.IsLargeBody {type=NiagaraBool}
 ]
}
